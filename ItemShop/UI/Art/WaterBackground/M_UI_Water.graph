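# Translucent UI material that draws an animated water background, built with Item Shop tile functions for background colors and rounded corners.
# A panning flow-map texture offsets UVs for two differently scaled layers of one caustics texture.
# Panning blurred noise breaks up the top layer, while sine-eased curves modulate layer intensity.
# A radial gradient forms the backdrop, and a rotated reflected gradient with delayed, eased animation supplies a sweeping reflection mask.
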Material /ItemShop/UI/Art/WaterBackground/M_UI_Water
graph {
 properties {
  bUsedWithStaticMesh=false
  blendMode=BLEND_Translucent
  materialDomain=MD_UI
  pixelDepthOffsetMode=PDOM_Legacy
  usedShadingModels=Default Lit
 }
 nodes [
  n1 = NamedRerouteUsage()
  n2 = NamedRerouteUsage()
  Add(A=Multiply(A=NamedRerouteUsage(), B=Add(A=@n1, B=@n2)), B=NamedRerouteUsage())
  Constant3Vector() {constant=<rgba 0 0.5395 0.3712 0>}
  Constant3Vector() {constant=<rgba 0 0.0423 0.4735 0>}
  MaterialFunctionCall!(Input Alpha=Constant() {r=1}, bOverrideCornerRadius, Corner Radius Override) {materialFunction=/FortUILibrary/Systems/ItemShopTile/Materials/MaterialFunctions/MF_UI_ItemShop_RoundedCorners}
  n3 = MaterialFunctionCall!(ScreenUV, TextureSize, ScalePivotPoint) {materialFunction=/FortUILibrary/MaterialFunctions/Basics/MF_UI_ScreenSpaceUVScaling}
  NamedRerouteDeclaration(None=@n3.ScaleX) {name=ScaleX, nodeColor=<rgba 1 0 0.0938 1>}
  n4 = NamedRerouteUsage()
  n5 = MaterialFunctionCall(
   UV=NamedRerouteUsage()
   GradientStart=ScalarParameter() {defaultValue=-0.188}
   GradientEnd=ScalarParameter() {defaultValue=0.4774, parameterName=Param_1}
  ) {materialFunction=&MF_UI_RadialGradient}
  NamedRerouteDeclaration(None=Add(A=Multiply(A=@n4, B=@n5.Smooth), B=NamedRerouteUsage())) {name=RadialGradientBG, nodeColor=<rgba 0.5234 0 1 1>}
  n6 = Constant() {r=0.15}
  n7 = NamedRerouteUsage()
  n8 = Constant() {r=1}
  n9 = TextureSample(
   UVs=Panner(Coordinate=MaterialFunctionCall(UVs=@n7, Texture Scale=@n8) {materialFunction=&ScaleUVsByCenter}.UVs) {speedX=0.07, speedY=0.053}
   Tex=TextureObject() {samplerType=SAMPLERTYPE_LinearColor, texture=/Game/UI/Frontend/ItemShop/Art/WaterBackground/T_UI_WaterFlowMap}
  )
  NamedRerouteDeclaration(None=Multiply(A=Constant2Vector() {g=1, r=-1}, B=Multiply(A=@n6, B=@n9.R))) {name=FlowMap, nodeColor=<rgba 0 0.0391 1 1>}
  NamedRerouteDeclaration(None=Panner(Coordinate=NamedRerouteUsage()) {speedX=0.005, speedY=0.01}) {name=FlowmapMovement, nodeColor=<rgba 0.3594 1 0 1>}
  n10 = MaterialFunctionCall() {materialFunction=/FortUILibrary/Systems/ItemShopTile/Materials/MaterialFunctions/MF_UI_ItemShop_BackgroundParams}
  NamedRerouteDeclaration(None=@n10.Color 1) {name=color 1, nodeColor=<rgba 0.2422 0 1 1>}
  NamedRerouteDeclaration(None=@n10.Color 2) {name=color 2, nodeColor=<rgba 0.0781 1 0 1>}
  NamedRerouteDeclaration(None=@n10.Color 3) {name=color 3, nodeColor=<rgba 1 0 0.3984 1>}
  n11 = MaterialFunctionCall(UVs=Add(A=NamedRerouteUsage(), B=NamedRerouteUsage()), Texture Scale=ScalarParameter() {defaultValue=0.15, parameterName=Layer1Scale}) {materialFunction=&ScaleUVsByCenter}
  n12 = TextureObject() {samplerType=SAMPLERTYPE_Alpha, texture=&/T_UI_Caustics_ShopTile}
  n13 = Saturate(None=Multiply(A=TextureSample(UVs=@n11.UVs, Tex=@n12).R, B=ScalarParameter() {defaultValue=0.75, parameterName=FirstWaveOpacity}))
  n14 = Panner(Coordinate=NamedRerouteUsage()) {speedX=0.025, speedY=0.05}
  n15 = Multiply(A=SmoothStep(Value=TextureSample(UVs=@n14) {samplerType=SAMPLERTYPE_Grayscale, texture=&LowResBlurredNoise}.R) {constMin=0.3})
  n16 = NamedRerouteUsage()
  n17 = LinearInterpolate(Alpha=MaterialFunctionCall!("Value (0-1)"=OneMinus(None=@n16)) {materialFunction=/FortUILibrary/MaterialFunctions/Basics/MF_UI_EaseCurves_01_Sine}.In-Out) {constA=0.6}
  n18 = Add(A=Divide(A=Min(A=@n13, B=@n15)) {constB=0.3}, B=Multiply(A=Multiply(A=@n17, B=@n13)) {constB=0.5})
  n19 = Multiply(A=NamedRerouteUsage(), B=@n18)
  NamedRerouteDeclaration(None=Add(A=@n18, B=Divide(A=@n19) {constB=0.4})) {name=TopLayerCaustics, nodeColor=<rgba 0 1 0.2031 1>}
  n20 = Rotator(Coordinate=NamedRerouteUsage(), Time=Constant() {r=-0.1}) {speed=6.2832}
  n21 = Constant() {r=8}
  n22 = Constant() {r=2}
  n23 = LinearInterpolate(Alpha=MaterialFunctionCall("Value (0-1)"=MaterialFunctionCall!(Animation Duration=@n21, Delay=@n22, Reset to Default) {materialFunction=/FortUILibrary/MaterialFunctions/Time/MF_UI_AnimationDelay}.Result) {materialFunction=&MF_UI_EaseCurves}.EaseInCubic) {constA=-2, constB=2}
  NamedRerouteDeclaration(None=SmoothStep(Value=MaterialFunctionCall!(UVs=@n20, Offset=@n23, RotationCenter, RotationAngle) {materialFunction=/FortUILibrary/MaterialFunctions/Gradients/MF_UI_ReflectedGradient}.Result) {constMin=0.5}) {name=Reflection Mask, nodeColor=<rgba 0 0 0 1>}
  n24 = MaterialFunctionCall("Value (0-1)"=Abs(None=NamedRerouteUsage())) {materialFunction=&MF_UI_EaseCurves_01_Sine}
  n25 = Add(A=NamedRerouteUsage(), B=NamedRerouteUsage())
  n26 = ScalarParameter() {defaultValue=0.35, parameterName=Layer2Scale}
  n27 = Multiply(
   A=TextureSample(
    UVs=MaterialFunctionCall(UVs=@n25, Texture Scale=@n26) {materialFunction=&ScaleUVsByCenter}.UVs
    Tex=NamedRerouteUsage()
   ).R
   B=ScalarParameter() {defaultValue=0.4, parameterName=SecondaryWaveOpacity}
  )
  NamedRerouteDeclaration(None=Multiply(A=LinearInterpolate(Alpha=@n24.In-Out) {constA=0.3}, B=Saturate(None=@n27))) {name=BottomLayerCaustics, nodeColor=<rgba 1 0 0.1172 1>}
  n28 = Add(A=Multiply(A=Time(), B=Constant() {r=0.1}))
  NamedRerouteDeclaration(None=ConstantBiasScale(None=Frac(None=@n28)) {bias=-0.5, scale=2}) {name=FadeTime, nodeColor=<rgba 0 0.4375 1 1>}
  NamedRerouteDeclaration(None=@n3.Scaled to Fit) {name=ScaleToFit, nodeColor=<rgba 0.7578 1 0 1>}
  n29 = MaterialFunctionCall() {materialFunction=&MF_UI_UVScaling}
  NamedRerouteDeclaration(None=@n29.ScaleX) {name=ScaleXUVScale, nodeColor=<rgba 0.9219 0 1 1>}
  NamedRerouteDeclaration(None=@n29.Scaled to Fit) {name=ScaletoFitUVScale, nodeColor=<rgba 0 1 0.6016 1>}
  NamedRerouteDeclaration(None=@n12) {name=CausticTexture, nodeColor=<rgba 1 0.2812 0 1>}
  NamedRerouteUsage()
  VectorParameter() {parameterName=CausticColor, defaultValue=<rgba 0 0.4179 1 1>}
 ]
}
